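; Parent knockback gameplay ability for SinisterMelon that extends an inherited knockback ability.
; On activation, a hotfix flag gates teammate knockback; when off, it checks the team affiliation of avatar and effect causer so only neutral or hostile targets proceed.
; On launch, the authority applies a till-land fall damage immunity effect to the knocked player.
; The initial knock calculation can substitute the launcher's forward vector for the supplied direction.
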(event Ability|EventActivateAbility
  ;[0 0.15 0.0113 0.5] ActivateAbility: Knockback based on teammate status.
  (if (not (ScalableFloat|GetValueasBool (Variables|Hotfix|GetHotfix_AllowTeammateKnockback)))
    (switch Utilities|FlowControl|Switch|SwitchonEFortTeamAffiliation (Team|GetTeamAffiliation (Ability|GetAvatarActorfromActorInfo) (Ability|EffectContext|GetEffectCauser (Ability|GetGrantedbyEffectContext)))
      (:Neutral
        _)
      (:Hostile
        _)))
  (|CallParentFunction))

(event EventDoLaunch
  (params (Vel :type Vector :default "0, 0, 0"))
  ;[0 0.15 0.1463 0.5] DoLaunch: Grant fall damage immunity to the knocked player.
  (|CallParentFunction Vel)
  (Ability|HasAuthority
    (:True
      (Ability|ApplyGameplayEffectToOwner "/SinisterMelon/Gameplay/SinisterMelon/GameplayEffects/GE_Athena_SinisterMelon_TillLandFallDamageImmunity_Grant.GE_Athena_SinisterMelon_TillLandFallDamageImmunity_Grant_C"))))

(fn GetInitialKnockMag
  (params (Direction :type Vector :default "0, 0, 0") (OverrideXYMag :type Boolean) (XYMAgOverride :type Float%20%28double-precision%29))
  (if (Variables|Default|GetbKnockLauncherForward)
    (return (|CallParentFunction (Transformation|GetActorForwardVector (Variables|KnockbackParent|GetLauncher))))
    (else
      (return (|CallParentFunction Direction)))))
